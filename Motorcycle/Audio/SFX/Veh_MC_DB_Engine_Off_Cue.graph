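# Sound cue for the motorcycle dirt bike engine shutting off.
# A Random node picks between two engine-off wave variations from different dirt bike recordings with equal weight.
# Output uses shared vehicle movement attenuation, the exterior vehicle sound class, reduced volume and the vehicle engine concurrency group.
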SoundCue /Motorcycle/Audio/SFX/Veh_MC_DB_Engine_Off_Cue
root=Random_0
nodes [
 {id=Random_0, children [WavePlayer_0,WavePlayer_3], properties {weights [1,1]}}
 {id=WavePlayer_0, properties {soundWaveAssetPtr=&/Veh_MC_Dirtbike_02_Engine_Off_01}}
 {id=WavePlayer_3, properties {soundWaveAssetPtr=&/Veh_MC_Dirtbike_03_Engine_Off_01}}
]
properties {
 attenuationSettings=/VehicleAudio/Settings/Vehicle_Movement_Att
 maxDistance=12512
 soundClassObject=&Fort_Vehicles_Exterior
 volumeMultiplier=0.7
 concurrencySet [/VehicleAudio/Settings/Vehicle_Engine_Conc]
}
