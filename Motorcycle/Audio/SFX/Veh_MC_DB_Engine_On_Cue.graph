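# Sound cue for the dirt bike motorcycle's engine start, played as an exterior vehicle effect.
# A Random node picks with equal weight between two Dirtbike engine-on wave variants for variety.
# Vehicle movement attenuation, engine concurrency and the exterior vehicle sound class govern its falloff, voice limits and mixing, with output volume slightly reduced.
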SoundCue /Motorcycle/Audio/SFX/Veh_MC_DB_Engine_On_Cue
root=Random_0
nodes [
 {id=Random_0, children [WavePlayer_1,WavePlayer_2], properties {weights [1,1]}}
 {id=WavePlayer_1, properties {soundWaveAssetPtr=&/Veh_MC_Dirtbike_03_Engine_On_01}}
 {id=WavePlayer_2, properties {soundWaveAssetPtr=&/Veh_MC_Dirtbike_02_Engine_On_01}}
]
properties {
 attenuationSettings=/VehicleAudio/Settings/Vehicle_Movement_Att
 maxDistance=12512
 soundClassObject=&Fort_Vehicles_Exterior
 volumeMultiplier=0.7
 concurrencySet [/VehicleAudio/Settings/Vehicle_Engine_Conc]
}
