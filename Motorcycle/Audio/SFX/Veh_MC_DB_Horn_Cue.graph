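# Sound cue for the dirtbike motorcycle horn, routed to the exterior vehicles sound class with horn attenuation.
# A single horn wave player feeds a continuous modulator that maps the PitchOffset parameter to a narrow pitch range around normal.
# A concurrency override limits the cue to one instance, taking precedence over the assigned vehicle engine concurrency set.
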SoundCue /Motorcycle/Audio/SFX/Veh_MC_DB_Horn_Cue
root=ModulatorContinuous_0
nodes [
 {
  id=ModulatorContinuous_0
  children [WavePlayer_0]
  properties {pitchModulationParams {default=0, maxOutput=1.05, minOutput=0.95, parameterName=PitchOffset}}
 }
 {id=WavePlayer_0, properties {soundWaveAssetPtr=&/Veh_MC_Dirtbike_Horn_01}}
]
properties {
 attenuationSettings=/VehicleAudio/Settings/Vehicle_Horn_Att
 bOverrideConcurrency=true
 maxDistance=16030.001
 soundClassObject=&Fort_Vehicles_Exterior
 volumeMultiplier=0.951
 concurrencyOverrides {maxCount=1}
 concurrencySet [/VehicleAudio/Settings/Vehicle_Engine_Conc]
}
