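# Sound cue for dirtbike tire skids, with randomized dirt and asphalt variants, using vehicle movement attenuation.
# A quality-level root sends its first two slots to larger wave pools and its last two to reduced pools.
# Each branch applies slight pitch modulation, then switches on the Surface parameter to a random dirt or asphalt pool, leaving other surfaces and the default silent.
# Concurrency allows one instance per owner, stopping the oldest.
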SoundCue /Motorcycle/Audio/SFX/Veh_MC_Dirtbike_Tires_Skid_Cue
root=QualityLevel_0
nodes [
 {id=Mixer_0, children [Random_2,null], properties {inputVolume [1,1]}}
 {id=Mixer_1, children [Random_3,null], properties {inputVolume [1,1]}}
 {id=Mixer_2, children [Random_0,null], properties {inputVolume [1,1]}}
 {id=Mixer_3, children [Random_1,null], properties {inputVolume [1,1]}}
 {id=Modulator_0, children [Switch_0], properties {pitchMax=1.1, pitchMin=0.9}}
 {id=Modulator_1, children [Switch_1], properties {pitchMax=1.1, pitchMin=0.9}}
 {id=QualityLevel_0, children [Modulator_0,Modulator_0,Modulator_1,Modulator_1]}
 {
  id=Random_0
  children [WavePlayer_3,WavePlayer_4,WavePlayer_7,WavePlayer_8,WavePlayer_9]
  properties {bShouldExcludeFromBranchCulling=true, weights [1,1,1,1,1]}
 }
 {
  id=Random_1
  children [WavePlayer_39, WavePlayer_40, WavePlayer_41, WavePlayer_42]
  properties {bShouldExcludeFromBranchCulling=true, weights [1,1,1,1]}
 }
 {
  id=Random_2
  children [
   WavePlayer_0
   WavePlayer_1
   WavePlayer_2
   WavePlayer_12
   WavePlayer_13
   WavePlayer_14
   WavePlayer_15
   WavePlayer_16
   WavePlayer_17
   WavePlayer_18
   WavePlayer_19
   WavePlayer_20
   WavePlayer_21
   WavePlayer_22
  ]
  properties {bShouldExcludeFromBranchCulling=true, weights [1,1,1,1,1,1,1,1,1,1,1,1,1,1]}
 }
 {
  id=Random_3
  children [WavePlayer_23, WavePlayer_24, WavePlayer_25, WavePlayer_26, WavePlayer_27, WavePlayer_28, WavePlayer_29, WavePlayer_30, WavePlayer_31]
  properties {bShouldExcludeFromBranchCulling=true, weights [1,1,1,1,1,1,1,1,1]}
 }
 {id=Switch_0, children [null,Mixer_1,null,null,null,null,null,null,null,null,Mixer_0,Mixer_0], properties {intParameterName=Surface}}
 {id=Switch_1, children [null,Mixer_3,null,null,null,null,null,null,null,null,Mixer_2,Mixer_2], properties {intParameterName=Surface}}
 {id=WavePlayer_0, properties {soundWaveAssetPtr=&/Veh_MC_Dirtbike_Tire_Skid_Dirt_14}}
 {id=WavePlayer_1, properties {soundWaveAssetPtr=&/Veh_MC_Dirtbike_Tire_Skid_Dirt_01}}
 {id=WavePlayer_12, properties {soundWaveAssetPtr=&/Veh_MC_Dirtbike_Tire_Skid_Dirt_03}}
 {id=WavePlayer_13, properties {soundWaveAssetPtr=&/Veh_MC_Dirtbike_Tire_Skid_Dirt_04}}
 {id=WavePlayer_14, properties {soundWaveAssetPtr=&/Veh_MC_Dirtbike_Tire_Skid_Dirt_05}}
 {id=WavePlayer_15, properties {soundWaveAssetPtr=&/Veh_MC_Dirtbike_Tire_Skid_Dirt_06}}
 {id=WavePlayer_16, properties {soundWaveAssetPtr=&/Veh_MC_Dirtbike_Tire_Skid_Dirt_07}}
 {id=WavePlayer_17, properties {soundWaveAssetPtr=&/Veh_MC_Dirtbike_Tire_Skid_Dirt_08}}
 {id=WavePlayer_18, properties {soundWaveAssetPtr=&/Veh_MC_Dirtbike_Tire_Skid_Dirt_09}}
 {id=WavePlayer_19, properties {soundWaveAssetPtr=&/Veh_MC_Dirtbike_Tire_Skid_Dirt_10}}
 {id=WavePlayer_2, properties {soundWaveAssetPtr=&/Veh_MC_Dirtbike_Tire_Skid_Dirt_02}}
 {id=WavePlayer_20, properties {soundWaveAssetPtr=&/Veh_MC_Dirtbike_Tire_Skid_Dirt_11}}
 {id=WavePlayer_21, properties {soundWaveAssetPtr=&/Veh_MC_Dirtbike_Tire_Skid_Dirt_12}}
 {id=WavePlayer_22, properties {soundWaveAssetPtr=&/Veh_MC_Dirtbike_Tire_Skid_Dirt_13}}
 {id=WavePlayer_23, properties {soundWaveAssetPtr=&/Veh_MC_Dirtbike_Tire_Skid_Asphalt_01}}
 {id=WavePlayer_24, properties {soundWaveAssetPtr=&/Veh_MC_Dirtbike_Tire_Skid_Asphalt_02}}
 {id=WavePlayer_25, properties {soundWaveAssetPtr=&/Veh_MC_Dirtbike_Tire_Skid_Asphalt_03}}
 {id=WavePlayer_26, properties {soundWaveAssetPtr=&/Veh_MC_Dirtbike_Tire_Skid_Asphalt_04}}
 {id=WavePlayer_27, properties {soundWaveAssetPtr=&/Veh_MC_Dirtbike_Tire_Skid_Asphalt_05}}
 {id=WavePlayer_28, properties {soundWaveAssetPtr=&/Veh_MC_Dirtbike_Tire_Skid_Asphalt_06}}
 {id=WavePlayer_29, properties {soundWaveAssetPtr=&/Veh_MC_Dirtbike_Tire_Skid_Asphalt_07}}
 {id=WavePlayer_3, properties {soundWaveAssetPtr=&Veh_MC_Dirtbike_Tire_Skid_Dirt_14}}
 {id=WavePlayer_30, properties {soundWaveAssetPtr=&/Veh_MC_Dirtbike_Tire_Skid_Asphalt_08}}
 {id=WavePlayer_31, properties {soundWaveAssetPtr=&/Veh_MC_Dirtbike_Tire_Skid_Asphalt_09}}
 {id=WavePlayer_39, properties {soundWaveAssetPtr=&Veh_MC_Dirtbike_Tire_Skid_Asphalt_01}}
 {id=WavePlayer_4, properties {soundWaveAssetPtr=&Veh_MC_Dirtbike_Tire_Skid_Dirt_01}}
 {id=WavePlayer_40, properties {soundWaveAssetPtr=&Veh_MC_Dirtbike_Tire_Skid_Asphalt_02}}
 {id=WavePlayer_41, properties {soundWaveAssetPtr=&Veh_MC_Dirtbike_Tire_Skid_Asphalt_03}}
 {id=WavePlayer_42, properties {soundWaveAssetPtr=&Veh_MC_Dirtbike_Tire_Skid_Asphalt_04}}
 {id=WavePlayer_7, properties {soundWaveAssetPtr=&Veh_MC_Dirtbike_Tire_Skid_Dirt_02}}
 {id=WavePlayer_8, properties {soundWaveAssetPtr=&Veh_MC_Dirtbike_Tire_Skid_Dirt_03}}
 {id=WavePlayer_9, properties {soundWaveAssetPtr=&Veh_MC_Dirtbike_Tire_Skid_Dirt_04}}
]
properties {
 attenuationSettings=/VehicleAudio/Settings/Vehicle_Movement_Att
 bOverrideConcurrency=true
 maxDistance=12512
 soundClassObject=&Fort_Vehicles_Exterior
 volumeMultiplier=0.5
 concurrencyOverrides {bLimitToOwner=true, maxCount=1, resolutionRule=StopOldest, retriggerTime=1}
}
