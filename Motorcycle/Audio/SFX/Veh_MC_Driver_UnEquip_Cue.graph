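# Sound cue for the motorcycle driver unequip sound, built on a dirtbike driver exit recording.
# A single wave player feeds a Modulator at the root, which is left at its usual pitch and volume settings.
# The cue uses the shared close-detail vehicle movement attenuation and routes to the Fort_Vehicles sound class.
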SoundCue /Motorcycle/Audio/SFX/Veh_MC_Driver_UnEquip_Cue
root=Modulator_0
nodes [
 {id=Modulator_0, children [WavePlayer_2]}
 {id=WavePlayer_2, properties {soundWaveAssetPtr=&/Veh_MC_Dirtbike_Driver_Exit_01}}
]
properties {
 attenuationSettings=/VehicleAudio/Settings/Vehicle_Movement_Close_Detail_Att
 maxDistance=12000
 soundClassObject=&Fort_Vehicles
 volumeMultiplier=1
}
