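# Sound cue for the motorcycle jump charge, layering start, loop and completion sounds on one timeline.
# A Mixer plays the start one-shot immediately, the completion one-shot after a fixed delay, and a half-volume looping wave.
# An Enveloper silences the loop until the completion sound fires, then fades it in briefly.
# Uses vehicle boost attenuation, the Fort_Vehicles sound class and a concurrency override that stops the farthest instance.
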SoundCue /Motorcycle/Audio/SFX/Veh_MC_Jump_Charging_Loop_Cue
root=Mixer_1
nodes [
 {id=Delay_1, children [WavePlayer_3], properties {delayMax=1.25, delayMin=1.25}}
 {
  id=Enveloper_1
  children [WavePlayer_0]
  properties {pitchCurve {editorCurveData {keys [{time=0}]}}, volumeCurve {editorCurveData {keys [<at 0 0>,<at 1.25 0>,<at 1.6 1>]}}}
 }
 {id=Mixer_1, children [WavePlayer_1,Enveloper_1,Delay_1], properties {inputVolume [1,0.5,1]}}
 {id=WavePlayer_0, properties {bLooping=true, soundWaveAssetPtr=&/Veh_MC_Jump_Charge_Loop_01}}
 {id=WavePlayer_1, properties {soundWaveAssetPtr=&/Veh_MC_Jump_Charge_Start_01}}
 {id=WavePlayer_3, properties {soundWaveAssetPtr=&/Veh_MC_Jump_Charge_Complete_01}}
]
properties {
 attenuationSettings=/VehicleAudio/Settings/Vehicle_Movement_Boost_Att
 bOverrideConcurrency=true
 maxDistance=12000
 soundClassObject=&Fort_Vehicles
 volumeMultiplier=1.5
 attenuationOverrides {bEnableReverbSend=false, distanceAlgorithm=Logarithmic, falloffDistance=1000, stereoSpread=100, attenuationShapeExtents {x=125}}
 concurrencyOverrides {maxCount=2, resolutionRule=StopFarthestThenPreventNew}
}
