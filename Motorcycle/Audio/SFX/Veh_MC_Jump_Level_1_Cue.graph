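# Sound cue for the low-intensity (level 1) motorcycle jump sound effect.
# A Random node picks one of four equally weighted low jump waves and feeds it through a Modulator for variation.
# It plays in the vehicle sound class with shared vehicle movement attenuation and primes on load.
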SoundCue /Motorcycle/Audio/SFX/Veh_MC_Jump_Level_1_Cue
root=Modulator_1
nodes [
 {id=Modulator_1, children [Random_0]}
 {id=Random_0, children [WavePlayer_0,WavePlayer_1,WavePlayer_4,WavePlayer_5], properties {weights [1,1,1,1]}}
 {id=WavePlayer_0, properties {soundWaveAssetPtr=&/Veh_MC_Jump_Low_01}}
 {id=WavePlayer_1, properties {soundWaveAssetPtr=&/Veh_MC_Jump_Low_02}}
 {id=WavePlayer_4, properties {soundWaveAssetPtr=&/Veh_MC_Jump_Low_03}}
 {id=WavePlayer_5, properties {soundWaveAssetPtr=&/Veh_MC_Jump_Low_04}}
]
properties {
 attenuationSettings=/VehicleAudio/Settings/Vehicle_Movement_Att
 bPrimeOnLoad=true
 maxDistance=12512
 soundClassObject=&Fort_Vehicles
 volumeMultiplier=1
}
