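# Sound cue for the motorcycle's level 2 jump sound effect, in the vehicles sound class.
# A Random node picks one of four equally weighted high jump waves, which then passes through a Modulator for variation.
# It primes on load and uses a distant golf cart movement attenuation asset for spatial falloff.
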SoundCue /Motorcycle/Audio/SFX/Veh_MC_Jump_Level_2_Cue
root=Modulator_1
nodes [
 {id=Modulator_1, children [Random_0]}
 {id=Random_0, children [WavePlayer_0,WavePlayer_1,WavePlayer_4,WavePlayer_5], properties {weights [1,1,1,1]}}
 {id=WavePlayer_0, properties {soundWaveAssetPtr=&/Veh_MC_Jump_High_04}}
 {id=WavePlayer_1, properties {soundWaveAssetPtr=&/Veh_MC_Jump_High_01}}
 {id=WavePlayer_4, properties {soundWaveAssetPtr=&/Veh_MC_Jump_High_02}}
 {id=WavePlayer_5, properties {soundWaveAssetPtr=&/Veh_MC_Jump_High_03}}
]
properties {
 attenuationSettings=/Game/Sounds/Attenuation/Fort_GolfCart_Movement_Distant_Att
 bPrimeOnLoad=true
 maxDistance=10100
 soundClassObject=&Fort_Vehicles
 volumeMultiplier=1
}
