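# Sound cue for motorcycle gear-shift backfire effects, routed to the vehicle interior sound class.
# A random node picks among four equally weighted backfire waves and is excluded from branch culling, feeding a modulator that varies pitch and volume.
# It uses close-detail vehicle movement attenuation at reduced volume.
# A concurrency override limits instances per owner, stopping the oldest when the cap is reached.
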SoundCue /Motorcycle/Audio/SFX/Veh_MC_Shift_Backfire_Cue
root=Modulator_0
nodes [
 {id=Modulator_0, children [Random_0], properties {pitchMax=1.1, volumeMin=0.8}}
 {
  id=Random_0
  children [WavePlayer_0,WavePlayer_1,WavePlayer_5,WavePlayer_7]
  properties {bShouldExcludeFromBranchCulling=true, weights [1,1,1,1]}
 }
 {id=WavePlayer_0, properties {soundWaveAssetPtr=&/Veh_MC_Shift_BackFire_01}}
 {id=WavePlayer_1, properties {soundWaveAssetPtr=&/Veh_MC_Shift_BackFire_02}}
 {id=WavePlayer_5, properties {soundWaveAssetPtr=&/Veh_MC_Shift_BackFire_03}}
 {id=WavePlayer_7, properties {soundWaveAssetPtr=&/Veh_MC_Shift_BackFire_05}}
]
properties {
 attenuationSettings=/VehicleAudio/Settings/Vehicle_Movement_Close_Detail_Att
 bOverrideConcurrency=true
 maxDistance=12000
 soundClassObject=&Fort_Vehicles_Interior
 volumeMultiplier=0.5
 concurrencyOverrides {bLimitToOwner=true, maxCount=2, resolutionRule=StopOldest}
}
