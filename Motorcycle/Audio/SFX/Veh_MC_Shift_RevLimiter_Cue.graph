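# Sound cue for a motorcycle rev-limiter shift effect, built from a single dirt bike backfire pop wave.
# A short random delay feeds a modulator that randomizes pitch and volume, so repeated pops vary in timing and tone.
# It plays quietly in the vehicle interior sound class with close-detail movement attenuation, and a concurrency override caps simultaneous instances with a brief retrigger time.
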SoundCue /Motorcycle/Audio/SFX/Veh_MC_Shift_RevLimiter_Cue
root=Delay_0
nodes [
 {id=Delay_0, children [Modulator_0], properties {delayMax=0.1}}
 {id=Modulator_0, children [WavePlayer_2], properties {pitchMax=1.25, pitchMin=0.9, volumeMax=1.2, volumeMin=0.9}}
 {id=WavePlayer_2, properties {soundWaveAssetPtr=&/Veh_MC_DirtBike_Backfire_Pop_01}}
]
properties {
 attenuationSettings=/VehicleAudio/Settings/Vehicle_Movement_Close_Detail_Att
 bOverrideConcurrency=true
 maxDistance=12000
 soundClassObject=&Fort_Vehicles_Interior
 volumeMultiplier=0.35
 concurrencyOverrides {maxCount=2, retriggerTime=0.1}
}
