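# Sound cue for a motorcycle wheelie, layering a wheelie start sound with a dirtbike exhaust rattle.
# A mixer blends two modulated, equally weighted random branches, one picking a start variant and the other a slightly quieter rattle with wider pitch and volume variation.
# It plays in the vehicles sound class with the shared movement boost attenuation asset.
# Its concurrency override blocks new instances once the limit is reached.
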SoundCue /Motorcycle/Audio/SFX/Veh_MC_Wheelie_Cue
root=Mixer_0
nodes [
 {id=Mixer_0, children [Modulator_1,Modulator_0], properties {inputVolume [1,0.9]}}
 {id=Modulator_0, children [Random_2], properties {pitchMax=1.1, pitchMin=0.9, volumeMax=1, volumeMin=0.7}}
 {id=Modulator_1, children [Random_0]}
 {id=Random_0, children [WavePlayer_0,WavePlayer_2,WavePlayer_3,WavePlayer_4,WavePlayer_5], properties {weights [1,1,1,1,1]}}
 {
  id=Random_2
  children [WavePlayer_10, WavePlayer_11, WavePlayer_12, WavePlayer_13]
  properties {bShouldExcludeFromBranchCulling=true, weights [1,1,1,1]}
 }
 {id=WavePlayer_0, properties {soundWaveAssetPtr=&/Veh_MC_Wheelie_Start_01}}
 {id=WavePlayer_10, properties {soundWaveAssetPtr=&/Veh_MC_Dirtbike_Wheelie_Exhaust_Rattle_01}}
 {id=WavePlayer_11, properties {soundWaveAssetPtr=&/Veh_MC_Dirtbike_Wheelie_Exhaust_Rattle_02}}
 {id=WavePlayer_12, properties {soundWaveAssetPtr=&/Veh_MC_Dirtbike_Wheelie_Exhaust_Rattle_03}}
 {id=WavePlayer_13, properties {soundWaveAssetPtr=&/Veh_MC_Dirtbike_Wheelie_Exhaust_Rattle_04}}
 {id=WavePlayer_2, properties {soundWaveAssetPtr=&/Veh_MC_Wheelie_Start_02}}
 {id=WavePlayer_3, properties {soundWaveAssetPtr=&/Veh_MC_Wheelie_Start_03}}
 {id=WavePlayer_4, properties {soundWaveAssetPtr=&/Veh_MC_Wheelie_Start_04}}
 {id=WavePlayer_5, properties {soundWaveAssetPtr=&/Veh_MC_Wheelie_Start_05}}
]
properties {
 attenuationSettings=/VehicleAudio/Settings/Vehicle_Movement_Boost_Att
 bOverrideConcurrency=true
 maxDistance=12000
 soundClassObject=&Fort_Vehicles
 volumeMultiplier=1
 attenuationOverrides {bEnableReverbSend=false, distanceAlgorithm=Logarithmic, falloffDistance=1000, stereoSpread=100, attenuationShapeExtents {x=125}}
 concurrencyOverrides {resolutionRule=PreventNew}
}
