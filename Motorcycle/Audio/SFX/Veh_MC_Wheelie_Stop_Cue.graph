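# Sound cue for the motorcycle's wheelie stop effect, routed to the Fort_Vehicles sound class.
# An equally weighted Random node picks one of four Wheelie_End waves, and a Modulator adds slight pitch variation.
# The cue lowers the overall pitch and uses the shared vehicle movement boost attenuation asset.
# Its own concurrency override blocks new instances once the limit is reached.
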SoundCue /Motorcycle/Audio/SFX/Veh_MC_Wheelie_Stop_Cue
root=Modulator_0
nodes [
 {id=Modulator_0, children [Random_1], properties {pitchMax=1.1, pitchMin=0.9}}
 {id=Random_1, children [WavePlayer_1,WavePlayer_3,WavePlayer_4,WavePlayer_5], properties {weights [1,1,1,1]}}
 {id=WavePlayer_1, properties {soundWaveAssetPtr=&/Veh_MC_Wheelie_End_04}}
 {id=WavePlayer_3, properties {soundWaveAssetPtr=&/Veh_MC_Wheelie_End_01}}
 {id=WavePlayer_4, properties {soundWaveAssetPtr=&/Veh_MC_Wheelie_End_02}}
 {id=WavePlayer_5, properties {soundWaveAssetPtr=&/Veh_MC_Wheelie_End_03}}
]
properties {
 attenuationSettings=/VehicleAudio/Settings/Vehicle_Movement_Boost_Att
 bOverrideConcurrency=true
 maxDistance=12000
 pitchMultiplier=0.7
 soundClassObject=&Fort_Vehicles
 volumeMultiplier=1
 attenuationOverrides {bEnableReverbSend=false, distanceAlgorithm=Logarithmic, falloffDistance=1000, stereoSpread=100, attenuationShapeExtents {x=125}}
 concurrencyOverrides {resolutionRule=PreventNew}
}
